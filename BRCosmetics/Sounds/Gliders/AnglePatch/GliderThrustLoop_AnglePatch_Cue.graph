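# Sound cue for the looping thrust audio of the AnglePatch glider cosmetic, using glider thrust attenuation.
# A mixer layers four looping waves (main, flame, forward flame, texture), each through its own continuous modulator.
# The AnyDirection parameter drives the main layer's pitch and volume, while Forwards drives the other three.
# This lets the thrust blend shift with glider movement at runtime.
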SoundCue /BRCosmetics/Sounds/Gliders/AnglePatch/GliderThrustLoop_AnglePatch_Cue
root=Mixer_0
nodes [
 {
  id=Mixer_0
  children [ModulatorContinuous_0, ModulatorContinuous_2, ModulatorContinuous_3, ModulatorContinuous_1]
  properties {inputVolume [0.8,0.8,0.5,1]}
 }
 {
  id=ModulatorContinuous_0
  children [WavePlayer_0]
  properties {
   pitchModulationParams {default=0, maxOutput=1.1, minOutput=1, parameterName=AnyDirection}
   volumeModulationParams {default=0, minOutput=0.6, parameterName=AnyDirection}
  }
 }
 {
  id=ModulatorContinuous_1
  children [WavePlayer_2]
  properties {
   pitchModulationParams {default=0, minOutput=1, parameterName=Forwards}
   volumeModulationParams {default=0, maxOutput=0.3, parameterName=Forwards}
  }
 }
 {
  id=ModulatorContinuous_2
  children [WavePlayer_1]
  properties {
   pitchModulationParams {default=0, minOutput=0.98, parameterName=Forwards}
   volumeModulationParams {default=0, minOutput=0.6, parameterName=Forwards}
  }
 }
 {
  id=ModulatorContinuous_3
  children [WavePlayer_3]
  properties {
   pitchModulationParams {default=0, maxOutput=1.1, minOutput=1, parameterName=Forwards}
   volumeModulationParams {default=0, maxOutput=0.8, minOutput=0.2, parameterName=Forwards}
  }
 }
 {id=WavePlayer_0, properties {bLooping=true, soundWaveAssetPtr=&/IP_Glider_AnglePatch_Thrust_LP_Main}}
 {id=WavePlayer_1, properties {bLooping=true, soundWaveAssetPtr=&/IP_Glider_AnglePatch_Thrust_LP_Flame}}
 {id=WavePlayer_2, properties {bLooping=true, soundWaveAssetPtr=&/IP_Glider_AnglePatch_Thrust_LP_Texture}}
 {
  id=WavePlayer_3
  properties {bLooping=true, soundWaveAssetPtr=&/IP_Glider_AnglePatch_Thrust_LP_Flame_Fwd}
 }
]
properties {
 attenuationSettings=&Fort_Player_Glider_Thrust_Athena_Att
 audioPropertiesSheet=&APS_Cosmetic_Glider_Loop
 maxDistance=12500
 soundClassObject=&Fort_Gliders_Land
 volumeMultiplier=0.7
 concurrencySet [null]
}
